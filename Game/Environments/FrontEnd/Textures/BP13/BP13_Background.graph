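# Unlit front-end background material that maps the T_protofrontend_07x1 backplate texture in screen space.
# Viewport UVs are scaled about their center by ScaleX and ScaleY, adjusted for the view's aspect ratio through If comparisons against ScaleX.
# A view-direction parallax branch, clamped by MaxShift and weighted by radial screen falloff, is built but never reaches the sampled UVs.
# The graph shows no connection from the texture sample to a material output.
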Material /Game/Environments/FrontEnd/Textures/BP13/BP13_Background
graph {
 properties {pixelDepthOffsetMode=PDOM_Legacy, shadingModel=MSM_Unlit, usedShadingModels=Unlit}
 nodes [
  Constant() {r=1}
  n1 = ScalarParameter() {defaultValue=2.1357, parameterName=ScaleX}
  n2 = ScalarParameter() {defaultValue=0.5286, parameterName=ScaleY}
  Multiply(A=@n1, B=@n2) {constA=1}
  n3 = Transform(None=Constant3Vector() {constant=<rgba 1 0 0 0>}) {transformSourceType=TRANSFORMSOURCE_View}
  n4 = Multiply(A=DotProduct(A=Constant3Vector() {constant=<rgba 1 0 0 0>}, B=@n3), B=ScalarParameter() {defaultValue=0.1429, parameterName=ShiftX})
  n5 = ScalarParameter() {defaultValue=0.0286, parameterName=MaxShift}
  n6 = Multiply(A=@n5) {constA=-1, constB=-1}
  n7 = Constant3Vector() {constant=<rgba 0 0 1 0>}
  n8 = Multiply(
   A=DotProduct(A=Constant3Vector() {constant=<rgba 0 0 1 0>}, B=Transform(None=@n7) {transformSourceType=TRANSFORMSOURCE_Camera})
   B=ScalarParameter() {defaultValue=-0.1714, parameterName=ShiftY}
  )
  n9 = ScreenPosition()
  n10 = Constant2Vector() {g=0.5, r=0.5}
  n11 = Power(
   Base=Divide(A=Distance(A=@n9.ViewportUV, B=@n10), B=ScalarParameter() {defaultValue=1.0117, parameterName=DistanceParalax})
   Exp=ScalarParameter() {defaultValue=2, parameterName=Falloff}
  )
  n12 = ScalarParameter() {parameterName=Min}
  Multiply(A=AppendVector(A=Clamp(None=@n4, Min=@n6, Max=@n5), B=Clamp(None=@n8, Min=@n6, Max=@n5)), B=Saturate(None=Add(A=@n11, B=@n12)))
  Multiply(A=@n3, B=Constant3Vector() {constant=<rgba 1 1 0 0>})
  n13 = ViewSize()
  n14 = Divide(A=ComponentMask(None=@n13) {r=true}, B=ComponentMask(None=@n13) {g=true})
  n15 = AppendVector(
   A=Divide(B=If(A=@n14, B=@n1, "A > B"=@n1, "A == B"=@n1, "A < B"=@n14))
   B=Multiply(B=If(A=@n14, B=@n1, "A > B"=@n14, "A == B"=@n14, "A < B"=@n1)) {constA=1}
  )
  n16 = AppendVector(A=@n1, B=@n2)
  TextureSample(UVs=MaterialFunctionCall(UVs=@n9.ViewportUV, Texture Scale=Multiply(A=@n15, B=@n16)) {materialFunction=&ScaleUVsByCenter}.UVs) {texture=/Game/Environments/FrontEnd/BackPlates/Textures/T_protofrontend_07x1}
 ]
}
